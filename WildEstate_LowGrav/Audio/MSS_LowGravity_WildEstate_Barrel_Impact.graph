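# One-shot mono MetaSound Source for a barrel impact in WildEstate_LowGrav, using small-object hit attenuation and a concurrency limit.
# On play, it layers two randomly picked waves from array variables, starting one immediately and the other after a short delay.
# Both layers share one random pitch shift and feed a mono mixer.
# A trigger accumulator signals completion once both players finish.
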MetaSoundSource /WildEstate_LowGrav/Audio/MSS_LowGravity_WildEstate_Barrel_Impact
inputs [UE.Source.OnPlay {type=Trigger}]
interfaces [UE.OutputFormat.Mono:1.0, UE.Source.OneShot:1.0, UE.Source:1.1]
outputs [
 "UE.OutputFormat.Mono.Audio:0" {type=Audio}
 UE.Source.OneShot.OnFinished {type=Trigger}
]
pages [Default {
  input_defaults {UE.Source.OnPlay=null}
  nodes [
   TriggerDelay_0 = UE.Trigger Delay(In=^UE.Source.OnPlay) {also {defaults {Delay Time=0.05}}}
   ArrayRandomGet_0 = Array.Random Get:WaveAsset:Array(Next=@TriggerDelay_0.Out, In Array=VariableAccessor.WaveAsset:Array(Variable=InitVariable.WaveAsset:Array().Variable).Value) {also {defaults {Weights []}}}
   n1 = UE.RandomFloat(Next=^UE.Source.OnPlay) {also {defaults {Max=2, Min=-2}}}
   WavePlayer_0 = UE.Wave Player:Mono(Play=@ArrayRandomGet_0.On Next, Wave Asset=@ArrayRandomGet_0.Value, Pitch Shift=@n1.Value) {also {version [1,1]}}
   ArrayRandomGet_1 = Array.Random Get:WaveAsset:Array(Next=^UE.Source.OnPlay, In Array=VariableAccessor.WaveAsset:Array(Variable=InitVariable.WaveAsset:Array().Variable).Value)
   WavePlayer_1 = UE.Wave Player:Mono(Play=@ArrayRandomGet_1.On Next, Wave Asset=@ArrayRandomGet_1.Value, Pitch Shift=@n1.Value) {also {version [1,1]}}
   AudioMixerAudioMixerMono2_0 = "AudioMixer.Audio Mixer (Mono, 2)"(In 1=@WavePlayer_0.Out Mono, In 0=@WavePlayer_1.Out Mono)
   TriggerAccumulatorTriggerAccumulate2_0 = "TriggerAccumulator.Trigger Accumulate (2)"(In 1=@WavePlayer_0.On Finished, In 0=@WavePlayer_1.On Finished)
  ]
  links [
   {dst="@UE.OutputFormat.Mono.Audio:0", src=AudioMixerAudioMixerMono2_0.Out}
   {dst="@UE.Source.OneShot.OnFinished", src=TriggerAccumulatorTriggerAccumulate2_0.Out}
  ]
 }]
properties {
 assetImportData=&self:AssetImportData
 attenuationSettings=/Game/Sounds/Attenuation/Fort_Demolition_Hit_Small_Sized_Object_Att
 maxDistance=4000
 sampleRate=48000
 soundClassObject=&Fort_Impacts
 volume=0.56
 concurrencySet [&/CONC_LowGravity_WildEstate_FourMax]
}
