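# Sound cue for the Wild Estate low-gravity activation effect, built from gravity sphere sounds.
# A mixer layers a random "enter" sound with a quieter, briefly delayed random one-shot, each layer passing through a modulator for variation.
# It uses status-effect activation attenuation and the gadgets/items sound class.
# A concurrency override allows a single instance and rejects new plays while one is active.
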SoundCue /WildEstate_LowGrav/Audio/SC_LowGravity_WildEstate_Activate
root=Mixer_0
nodes [
 {id=Delay_0, children [Modulator_1], properties {delayMax=0.2, delayMin=0.2}}
 {id=Mixer_0, children [Modulator_0,Delay_0], properties {inputVolume [1,0.5]}}
 {id=Modulator_0, children [Random_1]}
 {id=Modulator_1, children [Random_0]}
 {id=Random_0, children [WavePlayer_3,WavePlayer_4,WavePlayer_5], properties {weights [1,1,1]}}
 {id=Random_1, children [WavePlayer_6,WavePlayer_7], properties {weights [1,1]}}
 {id=WavePlayer_3, properties {soundWaveAssetPtr=&/fort_gravity_sphere_oneShot_03}}
 {id=WavePlayer_4, properties {soundWaveAssetPtr=&/fort_gravity_sphere_oneShot_02}}
 {id=WavePlayer_5, properties {soundWaveAssetPtr=&/fort_gravity_sphere_oneShot_01}}
 {id=WavePlayer_6, properties {soundWaveAssetPtr=&/fort_gravity_sphere_enternew_01}}
 {id=WavePlayer_7, properties {soundWaveAssetPtr=&/fort_gravity_sphere_enternew_02}}
]
properties {
 attenuationSettings=/Game/Sounds/Attenuation/Fort_Ability_Status_Effect_Activation_Att
 maxDistance=6000
 soundClassObject=&Fort_Gadgets_Items
 volumeMultiplier=0.4
 concurrencyOverrides {maxCount=1, resolutionRule=PreventNew, retriggerTime=0.3}
}
