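# Looping ambient Sound Cue, named for the Wild Estate low-gravity barrel, layering metal stress groans and knocks over a crystal hum.
# A Mixer blends one randomly chosen, pitch-varied crystal loop with randomized metal one-shots that repeat after random delays and are pitched up.
# A Quality Level node gives the first two levels nine metal variations and the last two only four.
# It uses shared large-prop attenuation and overrides concurrency to cap instances.
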SoundCue /WildEstate_LowGrav/Audio/SC_LowGravity_WildEstate_Barrel_Loop
root=Mixer_0
nodes [
 {id=Delay_1, children [Random_2], properties {delayMax=3, delayMin=1}}
 {id=Delay_2, children [Random_4], properties {delayMax=3, delayMin=1}}
 {id=Looping_0, children [Delay_1]}
 {id=Looping_1, children [Delay_2]}
 {id=Mixer_0, children [Modulator_1,Modulator_0], properties {inputVolume [1,0.8]}}
 {id=Modulator_0, children [Random_0], properties {pitchMax=1.2, pitchMin=0.8}}
 {id=Modulator_1, children [QualityLevel_0], properties {pitchMax=1.3, pitchMin=1.2}}
 {id=QualityLevel_0, children [Looping_0,Looping_0,Looping_1,Looping_1]}
 {id=Random_0, children [WavePlayer_1,WavePlayer_2,WavePlayer_4], properties {weights [1,1,1]}}
 {
  id=Random_2
  children [WavePlayer_3, WavePlayer_5, WavePlayer_7, WavePlayer_8, WavePlayer_9, WavePlayer_10, WavePlayer_11, WavePlayer_12, WavePlayer_13]
  properties {weights [1,1,1,1,1,1,1,1,1]}
 }
 {id=Random_4, children [WavePlayer_14, WavePlayer_15, WavePlayer_16, WavePlayer_17], properties {weights [1,1,1,1]}}
 {id=WavePlayer_1, properties {bLooping=true, soundWaveAssetPtr=&/ZP_Crystal_Amb_Loop_01}}
 {id=WavePlayer_10, properties {soundWaveAssetPtr=&/AMB_Metal_Stress_Wronk_Groan_Knock_06}}
 {id=WavePlayer_11, properties {soundWaveAssetPtr=&/AMB_Metal_Stress_Wronk_Groan_Knock_07}}
 {id=WavePlayer_12, properties {soundWaveAssetPtr=&/AMB_Metal_Stress_Wronk_Groan_Knock_08}}
 {id=WavePlayer_13, properties {soundWaveAssetPtr=&/AMB_Metal_Stress_Wronk_Groan_Knock_09}}
 {id=WavePlayer_14, properties {soundWaveAssetPtr=&/AMB_Metal_Stress_Wronk_Groan_Knock_01}}
 {id=WavePlayer_15, properties {soundWaveAssetPtr=&/AMB_Metal_Stress_Wronk_Groan_Knock_02}}
 {id=WavePlayer_16, properties {soundWaveAssetPtr=&/AMB_Metal_Stress_Wronk_Groan_Knock_03}}
 {id=WavePlayer_17, properties {soundWaveAssetPtr=&/AMB_Metal_Stress_Wronk_Groan_Knock_04}}
 {id=WavePlayer_2, properties {bLooping=true, soundWaveAssetPtr=&/ZP_Crystal_Amb_Loop_02}}
 {id=WavePlayer_3, properties {soundWaveAssetPtr=&AMB_Metal_Stress_Wronk_Groan_Knock_01}}
 {id=WavePlayer_4, properties {bLooping=true, soundWaveAssetPtr=&/ZP_Crystal_Amb_Loop_03}}
 {id=WavePlayer_5, properties {soundWaveAssetPtr=&AMB_Metal_Stress_Wronk_Groan_Knock_02}}
 {id=WavePlayer_7, properties {soundWaveAssetPtr=&AMB_Metal_Stress_Wronk_Groan_Knock_03}}
 {id=WavePlayer_8, properties {soundWaveAssetPtr=&AMB_Metal_Stress_Wronk_Groan_Knock_04}}
 {id=WavePlayer_9, properties {soundWaveAssetPtr=&/AMB_Metal_Stress_Wronk_Groan_Knock_05}}
]
properties {
 attenuationSettings=/Game/Sounds/Attenuation/Fort_Amb_Emitter_Prop_Large_Att
 bOverrideConcurrency=true
 maxDistance=2000
 pitchMultiplier=1.5
 soundClassObject=&Fort_Ambient_Athena
 volumeMultiplier=0.3
 attenuationOverrides {
  bAttenuateWithLPF=true
  bEnableFocusInterpolation=true
  bEnableListenerFocus=true
  distanceAlgorithm=Logarithmic
  falloffDistance=2000
  focusAttackInterpSpeed=0.5
  focusAzimuth=15
  focusDistanceScale=0.5
  focusReleaseInterpSpeed=0.25
  lPFFrequencyAtMax=500
  lPFRadiusMax=2000
  lPFRadiusMin=1000
  nonFocusAzimuth=90
  nonFocusVolumeAttenuation=0.1
  stereoSpread=2048
  attenuationShapeExtents {x=500}
  customAttenuationCurve {editorCurveData {keys [<at 0 1>,<at 0.3 1>,<at 0.4 0.5>,<at 1 0>]}}
 }
 concurrencyOverrides {maxCount=3}
}
