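# Sound cue for low-gravity deactivation in the WildEstate LowGrav content, played from gravity sphere exit waves.
# A Random node picks between two exit wave variations with equal weight and feeds a Modulator at the root.
# The cue plays quietly in the gadgets and items sound class, using the shared ability status-effect activation attenuation.
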SoundCue /WildEstate_LowGrav/Audio/SC_LowGravity_WildEstate_Deactivate
root=Modulator_1
nodes [
 {id=Modulator_1, children [Random_0]}
 {id=Random_0, children [WavePlayer_0,WavePlayer_1], properties {weights [1,1]}}
 {id=WavePlayer_0, properties {soundWaveAssetPtr=&/fort_gravity_sphere_exitnew_01}}
 {id=WavePlayer_1, properties {soundWaveAssetPtr=&/fort_gravity_sphere_exitnew_03}}
]
properties {
 attenuationSettings=/Game/Sounds/Attenuation/Fort_Ability_Status_Effect_Activation_Att
 maxDistance=6000
 soundClassObject=&Fort_Gadgets_Items
 volumeMultiplier=0.15
}
