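# Sound cue for the jump sound of the Wild Estate low-gravity effect, using gadget/item sound class, attenuation and a shared jump/land concurrency set.
# A mixer layers pitched-up chrome blob squishes, delayed mushroom-launcher shield sounds with a rising pitch envelope, and quieter delayed low-gravity stone jumps.
# Each layer picks random waves, with the squish pool switching between five and three variants by quality level.
# An outer modulator varies the final mix's pitch.
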SoundCue /WildEstate_LowGrav/Audio/SC_LowGravity_WildEstate_Jump
root=Modulator_2
nodes [
 {id=Delay_0, children [Enveloper_0], properties {delayMax=0.2, delayMin=0.1}}
 {id=Delay_1, children [Random_0], properties {delayMax=0.15}}
 {
  id=Enveloper_0
  children [Modulator_0]
  properties {pitchCurve {editorCurveData {keys [<at 0 0.8>,<at 0.5 1.5>]}}, volumeCurve {editorCurveData {keys [<at 0 1>]}}}
 }
 {id=Mixer_0, children [Modulator_1,Delay_0,Delay_1], properties {inputVolume [1,1,0.6]}}
 {id=Modulator_0, children [Random_2], properties {pitchMax=1.3, pitchMin=1, volumeMax=1, volumeMin=0.9}}
 {id=Modulator_1, children [QualityLevel_0], properties {pitchMax=1.5, pitchMin=1.2}}
 {id=Modulator_2, children [Mixer_0], properties {pitchMax=1.2, pitchMin=0.9}}
 {id=QualityLevel_0, children [Random_1,Random_1,Random_3,Random_3]}
 {id=Random_0, children [WavePlayer_6,WavePlayer_7,WavePlayer_8], properties {weights [1,1,1]}}
 {id=Random_1, children [WavePlayer_3, WavePlayer_4, WavePlayer_5, WavePlayer_9, WavePlayer_10], properties {weights [1,1,1,1,1]}}
 {id=Random_2, children [WavePlayer_0,WavePlayer_1,WavePlayer_2], properties {weights [1,1,1]}}
 {id=Random_3, children [WavePlayer_11, WavePlayer_12, WavePlayer_13], properties {weights [1,1,1]}}
 {id=WavePlayer_0, properties {soundWaveAssetPtr=&/Env_Mushroom_Launcher_ShieldInc_01}}
 {id=WavePlayer_1, properties {soundWaveAssetPtr=&/Env_Mushroom_Launcher_ShieldInc_02}}
 {id=WavePlayer_10, properties {soundWaveAssetPtr=&/Chrome_Blob_Jump_Dash_Squish_3P_01}}
 {id=WavePlayer_11, properties {soundWaveAssetPtr=&/Chrome_Blob_Jump_Dash_Squish_3P_03}}
 {id=WavePlayer_12, properties {soundWaveAssetPtr=&/Chrome_Blob_Jump_Dash_Squish_3P_02}}
 {id=WavePlayer_13, properties {soundWaveAssetPtr=&Chrome_Blob_Jump_Dash_Squish_3P_01}}
 {id=WavePlayer_2, properties {soundWaveAssetPtr=&/Env_Mushroom_Launcher_ShieldInc_04}}
 {id=WavePlayer_3, properties {soundWaveAssetPtr=&/Chrome_Blob_Jump_Dash_Squish_3P_05}}
 {id=WavePlayer_4, properties {soundWaveAssetPtr=&/Chrome_Blob_Jump_Dash_Squish_3P_04}}
 {id=WavePlayer_5, properties {soundWaveAssetPtr=&Chrome_Blob_Jump_Dash_Squish_3P_03}}
 {id=WavePlayer_6, properties {soundWaveAssetPtr=&/LowGravityStone_Jump_03}}
 {id=WavePlayer_7, properties {soundWaveAssetPtr=&/LowGravityStone_Jump_02}}
 {id=WavePlayer_8, properties {soundWaveAssetPtr=&/LowGravityStone_Jump_01}}
 {id=WavePlayer_9, properties {soundWaveAssetPtr=&Chrome_Blob_Jump_Dash_Squish_3P_02}}
]
properties {
 attenuationSettings=/Game/Sounds/Attenuation/Fort_Ability_Status_Effect_Activation_Att
 maxDistance=6000
 pitchMultiplier=1.2
 soundClassObject=&Fort_Gadgets_Items
 volumeMultiplier=0.3
 concurrencySet [/Game/Sounds/Fort_GamePlay_Sounds/Player/LowGravityStones/LowGravityStone_JumpLand_Concc]
}
