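# Sound cue for the Wild Estate low-gravity jump "Stimpulse" sound, layering a jump footstep with a low-gravity stone effect.
# A Mixer blends a louder random chrome-footstep jump, pitched well down, with a quieter random LowGravityStone jump.
# Each layer has its own pitch-randomizing Modulator, and a root Modulator adds further variation to the mix.
# It uses status-effect activation attenuation and the shared LowGravityStone jump/land concurrency.
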SoundCue /WildEstate_LowGrav/Audio/SC_LowGravity_WildEstate_Jump_Stimpulse
root=Modulator_2
nodes [
 {id=Mixer_0, children [Modulator_3,Modulator_1], properties {inputVolume [1.4,0.6]}}
 {id=Modulator_1, children [Random_0], properties {pitchMax=1.2, pitchMin=0.9}}
 {id=Modulator_2, children [Mixer_0], properties {pitchMax=1.2, pitchMin=1}}
 {id=Modulator_3, children [Random_4], properties {pitchMax=0.7, pitchMin=0.5}}
 {id=Random_0, children [WavePlayer_18, WavePlayer_19], properties {weights [1,1]}}
 {id=Random_4, children [WavePlayer_14, WavePlayer_15, WavePlayer_16, WavePlayer_17], properties {weights [1,1,1,1]}}
 {id=WavePlayer_14, properties {soundWaveAssetPtr=/Game/Athena/Sounds/Character/Footsteps/1P/Chrome/Foley_FS_1P_Chrome_Jump_02}}
 {id=WavePlayer_15, properties {soundWaveAssetPtr=/Game/Athena/Sounds/Character/Footsteps/1P/Chrome/Foley_FS_1P_Chrome_Jump_03}}
 {id=WavePlayer_16, properties {soundWaveAssetPtr=/Game/Athena/Sounds/Character/Footsteps/1P/Chrome/Foley_FS_1P_Chrome_Jump_05}}
 {id=WavePlayer_17, properties {soundWaveAssetPtr=/Game/Athena/Sounds/Character/Footsteps/1P/Chrome/Foley_FS_1P_Chrome_Jump_08}}
 {id=WavePlayer_18, properties {soundWaveAssetPtr=&/LowGravityStone_Jump_06}}
 {id=WavePlayer_19, properties {soundWaveAssetPtr=&/LowGravityStone_Jump_05}}
]
properties {
 attenuationSettings=/Game/Sounds/Attenuation/Fort_Ability_Status_Effect_Activation_Att
 maxDistance=6000
 pitchMultiplier=1.2
 soundClassObject=&Fort_Gadgets_Items
 volumeMultiplier=0.6
 concurrencySet [/Game/Sounds/Fort_GamePlay_Sounds/Player/LowGravityStones/LowGravityStone_JumpLand_Concc]
}
