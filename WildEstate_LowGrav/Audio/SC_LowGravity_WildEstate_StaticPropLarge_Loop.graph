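# Looping ambient SoundCue for large static props in WildEstate low gravity, routed to the Athena ambient sound class.
# A Mixer layers a tower-base bed loop, through its own huge-prop attenuation node, with a quieter distant-activity loop randomly pitched down and volume-varied by a Modulator.
# The cue uses large-prop attenuation, raises overall pitch, lowers volume and overrides concurrency to cap simultaneous instances at three.
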SoundCue /WildEstate_LowGrav/Audio/SC_LowGravity_WildEstate_StaticPropLarge_Loop
root=Mixer_0
nodes [
 {id=Attenuation_0, children [WavePlayer_0], properties {attenuationSettings=/Game/Sounds/Attenuation/Fort_Amb_Emitter_Prop_Huge_Att}}
 {id=Mixer_0, children [Attenuation_0, Modulator_2], properties {inputVolume [1,0.7]}}
 {id=Modulator_2, children [WavePlayer_6], properties {pitchMax=0.8, pitchMin=0.7, volumeMax=1, volumeMin=0.8}}
 {id=WavePlayer_0, properties {bLooping=true, soundWaveAssetPtr=&/AMB_Bed_ZP_S16_Tower_Base_Loop_01}}
 {id=WavePlayer_6, properties {bLooping=true, soundWaveAssetPtr=&/ZP_Medium_Activity_Hi_End_Distant}}
]
properties {
 attenuationSettings=/Game/Sounds/Attenuation/Fort_Amb_Emitter_Prop_Large_Att
 bOverrideConcurrency=true
 maxDistance=10000
 pitchMultiplier=1.2
 soundClassObject=&Fort_Ambient_Athena
 volumeMultiplier=0.6
 attenuationOverrides {
  bAttenuateWithLPF=true
  bEnableFocusInterpolation=true
  bEnableListenerFocus=true
  distanceAlgorithm=Logarithmic
  falloffDistance=2000
  focusAttackInterpSpeed=0.5
  focusAzimuth=15
  focusDistanceScale=0.5
  focusReleaseInterpSpeed=0.25
  lPFFrequencyAtMax=500
  lPFRadiusMax=2000
  lPFRadiusMin=1000
  nonFocusAzimuth=90
  nonFocusVolumeAttenuation=0.1
  stereoSpread=2048
  attenuationShapeExtents {x=500}
  customAttenuationCurve {editorCurveData {keys [<at 0 1>,<at 0.3 1>,<at 0.4 0.5>,<at 1 0>]}}
 }
 concurrencyOverrides {maxCount=3}
}
