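# Looping ambient SoundCue for a static prop in the WildEstate low-gravity content, on the Athena ambient sound class.
# A Mixer layers a tower-base bed loop with a quieter crystal loop, randomly picked from three and pitched down by a Modulator.
# The bed has its own Huge-prop Attenuation node, while the cue uses the Large-prop attenuation asset.
# A concurrency override caps simultaneous instances at a few.
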SoundCue /WildEstate_LowGrav/Audio/SC_LowGravity_WildEstate_StaticProp_Loop
root=Mixer_0
nodes [
 {id=Attenuation_0, children [WavePlayer_0], properties {attenuationSettings=/Game/Sounds/Attenuation/Fort_Amb_Emitter_Prop_Huge_Att}}
 {id=Mixer_0, children [Attenuation_0, Modulator_0], properties {inputVolume [1,0.6]}}
 {id=Modulator_0, children [Random_0], properties {pitchMax=0.6, pitchMin=0.5}}
 {id=Random_0, children [WavePlayer_1,WavePlayer_2,WavePlayer_4], properties {weights [1,1,1]}}
 {id=WavePlayer_0, properties {bLooping=true, soundWaveAssetPtr=&/AMB_Bed_ZP_S16_Tower_Base_Loop_01}}
 {id=WavePlayer_1, properties {bLooping=true, soundWaveAssetPtr=&/ZP_Crystal_Amb_Loop_01}}
 {id=WavePlayer_2, properties {bLooping=true, soundWaveAssetPtr=&/ZP_Crystal_Amb_Loop_02}}
 {id=WavePlayer_4, properties {bLooping=true, soundWaveAssetPtr=&/ZP_Crystal_Amb_Loop_03}}
]
properties {
 attenuationSettings=/Game/Sounds/Attenuation/Fort_Amb_Emitter_Prop_Large_Att
 bOverrideConcurrency=true
 maxDistance=10000
 pitchMultiplier=1.2
 soundClassObject=&Fort_Ambient_Athena
 volumeMultiplier=0.6
 attenuationOverrides {
  bAttenuateWithLPF=true
  bEnableFocusInterpolation=true
  bEnableListenerFocus=true
  distanceAlgorithm=Logarithmic
  falloffDistance=2000
  focusAttackInterpSpeed=0.5
  focusAzimuth=15
  focusDistanceScale=0.5
  focusReleaseInterpSpeed=0.25
  lPFFrequencyAtMax=500
  lPFRadiusMax=2000
  lPFRadiusMin=1000
  nonFocusAzimuth=90
  nonFocusVolumeAttenuation=0.1
  stereoSpread=2048
  attenuationShapeExtents {x=500}
  customAttenuationCurve {editorCurveData {keys [<at 0 1>,<at 0.3 1>,<at 0.4 0.5>,<at 1 0>]}}
 }
 concurrencyOverrides {maxCount=3}
}
